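# Translucent unlit particle material, named for the Cratebot revive eye glow, for sprite and mesh particles.
# Emissive blends particle color with time-of-day tints and a curve-atlas gradient driven by height relative to object position, then runs through an emissive adjustment function.
# Opacity multiplies particle alpha by a texture or spherical mask, optional Fresnel camera fade and depth fade, which mobile skips.
# A static switch can restrict visibility to mobile.
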Material /Game/Characters/NPCs/Cratebot/Effects/M_Cratebot_ReviveEyeGlow
graph {
 properties {
  bUsedWithMeshParticles=true
  bUsedWithNiagaraMeshParticles=true
  bUsedWithNiagaraSprites=true
  bUsedWithParticleSprites=true
  blendMode=BLEND_Translucent
  pixelDepthOffsetMode=PDOM_Legacy
  shadingModel=MSM_Unlit
  usedShadingModels=Unlit
 }
 nodes [
  n1 = ParticleColor()
  n2 = MaterialFunctionCall() {materialFunction=/Game/Effects/Fort_Effects/Materials/_Functions/CurveColor_Time_of_Day}
  n3 = LinearInterpolate(
   A=@n1.RGB
   B=Multiply(A=VectorParameter() {group=TimeOfDay, parameterName=Time of Day Color, defaultValue=<rgba 1 1 1 1>}.RGB, B=Multiply(A=@n2.Result, B=@n1.RGB))
   Alpha=ScalarParameter() {group=TimeOfDay, parameterName=" Switch Time of Day"}
  )
  n4 = MaterialFunctionCall(
   Color_Morning=VectorParameter() {group=TOD, parameterName=Morning, defaultValue=<rgba 0 0 0 1>}.RGB
   Color_Day=VectorParameter() {group=TOD, parameterName=Day, defaultValue=<rgba 0 0 0 1>}.RGB
   Color_Evening=VectorParameter() {group=TOD, parameterName=Evening, defaultValue=<rgba 0 0 0 1>}.RGB
   Color_Night=VectorParameter() {group=TOD, parameterName=Night, defaultValue=<rgba 0 0 0 1>}.RGB
  ) {materialFunction=&Color_Time_of_Day}
  n5 = ScalarParameter() {group=TOD, parameterName=Legacy_TOD}
  n6 = ObjectPositionWS()
  n7 = Subtract(
   A=TransformPosition(None=WorldPosition() {worldPositionShaderOffset=WPT_ExcludeAllShaderOffsets}.XYZ) {transformSourceType=TRANSFORMPOSSOURCE_World}
   B=Subtract(A=TransformPosition(None=@n6) {transformSourceType=TRANSFORMPOSSOURCE_World}, B=Constant() {r=100})
  )
  n8 = CurveAtlasRowParameter(CurveTime=Divide(A=ComponentMask(None=@n7) {b=true}, B=Constant() {r=150}) {constB=1}) {
   atlas=&AthenaCurveAtlas01
   curve=/Game/Athena/LinearColorCurves/Curve_JellyfishPickaxe_Swing2
   defaultValue=100
  }
  n9 = ScalarParameter() {parameterName=MultiplyColor}
  MaterialFunctionCall(Emissive=LinearInterpolate(A=LinearInterpolate(A=@n3, B=@n4.Result, Alpha=@n5), B=Multiply(A=@n8, B=@n9), Alpha=ScalarParameter() {parameterName=World Position_Color})) {materialFunction=&MF_FX_EmissiveAdjustments}
  n10 = Constant()
  n11 = TextureSampleParameter2D() {parameterName=MaskTexture, texture=/Game/Global/Textures/Default/Masks_and_Gradients/T_Vert_Gradient_Soft}
  n12 = ScalarParameter() {defaultValue=1, group=CameraVectorFade, parameterName=Camera Vector Exponentln}
  n13 = Power(
   Base=OneMinus(None=Fresnel(ExponentIn=@n12) {baseReflectFraction=0, exponent=0.2})
   Exp=ScalarParameter() {defaultValue=2, group=CameraVectorFade, parameterName=Camera vector power}
  ) {constExponent=4}
  n14 = ScalarParameter() {group=CameraVectorFade, parameterName=Camera Vector Fade}
  n15 = LinearInterpolate(A=@n11.R, B=Multiply(A=@n11.R, B=@n13), Alpha=@n14)
  n16 = ShadingPathSwitch(Default=DepthFade(FadeDistance=ScalarParameter() {defaultValue=12, parameterName=FadeDistance}), Mobile=Constant() {r=1}) {desc=Migrated from FeatureLevelSwitch}
  n17 = TextureCoordinate()
  n18 = Constant() {r=0.5}
  n19 = Constant()
  n20 = Multiply(A=@n1.A, B=Power(Base=SphereMask(A=@n17, B=@n18, Radius=@n18, Hardness=@n19)))
  n21 = Multiply(A=@n20, B=@n13)
  n22 = StaticSwitchParameter(True=Multiply(A=@n1.A, B=Multiply(A=@n15, B=@n16)), False=Multiply(A=LinearInterpolate(A=@n20, B=@n21, Alpha=@n14), B=@n16)) {parameterName=UseMaskTexture?}
  StaticSwitchParameter(True=ShadingPathSwitch(Default=@n10, Deferred=@n10, Mobile=@n22) {desc=Migrated from FeatureLevelSwitch}, False=@n22) {parameterName=UseMoblieOnly?}
 ]
}
